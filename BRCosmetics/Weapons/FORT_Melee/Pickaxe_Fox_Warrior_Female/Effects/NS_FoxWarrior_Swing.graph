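# Niagara swing trail for the Fox Warrior Female pickaxe, built from three anim-trail ribbon emitters with short-lived particles that narrow and fade.
# User.VariantNum blends each trail from pink to blue.
# The main swiper trail ramps up brightness over its loop and computes pink and blue ribbon widths, zeroing one per variant, alongside two renderers.
# Two electric trails are named for PS5 and non-PS5, the PS5 one much dimmer.
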NiagaraSystem /BRCosmetics/Weapons/FORT_Melee/Pickaxe_Fox_Warrior_Female/Effects/NS_FoxWarrior_Swing
dependencies {
 usedDataInterfaces [&NiagaraDataInterfaceCurve, &NiagaraDataInterfaceArrayFloat]
 usedDynamicInputs []
 usedModules [&SystemState]
}
emitters {
 electric_feelings_NotPS5 {
  emitterUpdateScript [
   &EmitterState()
   &FN_SpawnAnimTrail_v2()
  ]
  particleSpawnScript [
   /Game/Effects/Niagara/Modules/AnimTrail/FN_InitializeAnimTrail/FN_InitializeAnimTrailBoneIndecies(Second Bone or Socket Index=1)
   /Game/Effects/Niagara/Modules/AnimTrail/FN_InitializeAnimTrail/FN_InitializeAnimTrail_v3/FN_InitializeAnimTrail(
    AnimTrailWidth=0.666
    Color:LinearColor=&MultiplyLinearColor(
     Linear Color=&LerpLinearColors(StartColor=<rgba 0.9216 0 0.1482 0.333>, EndColor=<rgba 0 0.1184 0.245 0.333>, LerpFactor=$User.VariantNum)
     Scale Factor=&MakeVector4_FromFloat(Float=25)
    )
    Lifetime=0.23
   )
  ]
  particleUpdateScript [
   /Game/Effects/Niagara/Modules/AnimTrail/FN_UpdateAnimTrail/FN_UpdateAnimTrail_v3/FN_UpdateAnimTrail()
   &ScaleRibbonWidth(Ribbon Width Scale=&FloatFromCurve(FloatCurve=DataInterface() {propertyValues {Curve {keys [
         {arriveTangent=-1.7147, interpMode=RCIM_Cubic, leaveTangent=-1.7147, tangentMode=RCTM_User, time=0, value=1}
         {arriveTangent=-0.2323, interpMode=RCIM_Cubic, leaveTangent=-0.2323, tangentMode=RCTM_User, time=1, value=0}
        ]}}}))
   &ScaleColor(
    Scale RGB=&VectorFromFloat(Value=&FloatFromCurve(FloatCurve=DataInterface() {propertyValues {Curve {keys [
          {arriveTangent=-3.243, interpMode=RCIM_Cubic, leaveTangent=-3.243, tangentMode=RCTM_User, time=0.25, value=15}
          {arriveTangent=-35.6813, interpMode=RCIM_Cubic, leaveTangent=-35.6813, tangentMode=RCTM_User, time=1, value=0}
         ]}}}))
    ScaleA=true
    ScaleRGB=true
   )
  ]
  rendererClasses [&NiagaraRibbonRendererProperties]
  renderers [{rendererClass=&NiagaraRibbonRendererProperties}]
 }
 electric_feelings_PS5 {
  emitterUpdateScript [
   &EmitterState()
   &FN_SpawnAnimTrail_v2()
  ]
  particleSpawnScript [
   &FN_InitializeAnimTrailBoneIndecies(Second Bone or Socket Index=1)
   &FN_InitializeAnimTrail(
    AnimTrailWidth=0.666
    Color:LinearColor=&MultiplyLinearColor(
     Linear Color=&LerpLinearColors(StartColor=<rgba 0.9216 0 0.1482 0.333>, EndColor=<rgba 0 0.1184 0.245 0.333>, LerpFactor=$User.VariantNum)
     Scale Factor=&MakeVector4_FromFloat(Float=5)
    )
    Lifetime=0.23
   )
  ]
  particleUpdateScript [
   &FN_UpdateAnimTrail()
   &ScaleRibbonWidth(Ribbon Width Scale=&FloatFromCurve(FloatCurve=DataInterface() {propertyValues {Curve {keys [
         {arriveTangent=-1.7147, interpMode=RCIM_Cubic, leaveTangent=-1.7147, tangentMode=RCTM_User, time=0, value=1}
         {arriveTangent=-0.2323, interpMode=RCIM_Cubic, leaveTangent=-0.2323, tangentMode=RCTM_User, time=1, value=0}
        ]}}}))
   &ScaleColor(
    Scale RGB=&VectorFromFloat(Value=&FloatFromCurve(FloatCurve=DataInterface() {propertyValues {Curve {keys [{arriveTangent=-3.243, interpMode=RCIM_Cubic, time=0.25, value=1}, {interpMode=RCIM_Cubic, leaveTangent=-35.6813, time=1, value=0}]}}}))
    ScaleA=true
    ScaleRGB=true
   )
  ]
  rendererClasses [&NiagaraRibbonRendererProperties]
  renderers [{rendererClass=&NiagaraRibbonRendererProperties}]
 }
 swiper_no_swiping {
  emitterUpdateScript [
   &EmitterState()
   &FN_SpawnAnimTrail_v2(Spawn Interval=0)
  ]
  particleSpawnScript [
   &FN_InitializeAnimTrailBoneIndecies(Second Bone or Socket Index=1)
   &FN_InitializeAnimTrail(
    AnimTrailWidth=0.6
    Color:LinearColor=&MultiplyLinearColor(
     Linear Color=&LerpLinearColors(StartColor=<rgba 0.9216 0.0021 0.15 0.333>, EndColor=<rgba 0 0.1833 1 0.333>, LerpFactor=$User.VariantNum)
     Scale Factor=&MakeVector4_FromFloat(Float=&FloatFromCurve(
       FloatCurve=DataInterface() {propertyValues {Curve {keys [{arriveTangent=3.4463, interpMode=RCIM_Cubic, time=0, value=0}, <at 0.2 1>, <at 1 1>]}}}
       CurveIndex=$Emitter.NormalizedLoopAge
      ))
    )
    Lifetime=0.165
   )
  ]
  particleUpdateScript [
   &FN_UpdateAnimTrail()
   &ScaleRibbonWidth(Ribbon Width Scale=&FloatFromCurve(FloatCurve=DataInterface() {propertyValues {Curve {keys [
         <at 0 1>
         {arriveTangent=-1, interpMode=RCIM_Cubic, leaveTangent=-1, time=0.2628, value=1}
         {arriveTangent=-2.5856, interpMode=RCIM_Cubic, leaveTangent=-2.5856, tangentMode=RCTM_User, time=1, value=0}
        ]}}}))
   &ScaleColor(
    Scale RGB=&VectorFromFloat(Value=&FloatFromCurve(FloatCurve=DataInterface() {propertyValues {Curve {keys [
          {arriveTangent=-3.243, interpMode=RCIM_Cubic, leaveTangent=-3.243, tangentMode=RCTM_User, time=0.25, value=15}
          {arriveTangent=-35.6813, interpMode=RCIM_Cubic, leaveTangent=-35.6813, tangentMode=RCTM_User, time=1, value=0}
         ]}}}))
    ScaleA=true
    ScaleRGB=true
   )
   /SetVariables(
    Particles.RibbonWidthBlue:NiagaraFloat=&Multiply_Float(
     A=$Particles.RibbonWidth
     B=&SelectFloatFromArray(
      Direct Array Index=&MakeIntFromFloat(Float=$User.VariantNum)
      Float Selection Array=DataInterface() {propertyValues {FloatData [0,1]}}
     )
    )
    Particles.RibbonWidthPink:NiagaraFloat=&Multiply_Float(
     A=$Particles.RibbonWidth
     B=&SelectFloatFromArray(
      Direct Array Index=&MakeIntFromFloat(Float=$User.VariantNum)
      Float Selection Array=DataInterface() {propertyValues {FloatData [1,0]}}
     )
    )
   ) {isSetParameters=true}
  ]
  rendererClasses [&NiagaraRibbonRendererProperties]
  renderers [
   {rendererClass=&NiagaraRibbonRendererProperties}
   {rendererClass=&NiagaraRibbonRendererProperties, rendererIndex=1}
  ]
 }
}
summary {
 systemName=NS_FoxWarrior_Swing
 emitters [
  {emitterName=swiper_no_swiping, rendererClasses [&NiagaraRibbonRendererProperties]}
  {emitterName=electric_feelings_NotPS5, rendererClasses [&NiagaraRibbonRendererProperties]}
  {emitterName=electric_feelings_PS5, rendererClasses [&NiagaraRibbonRendererProperties]}
 ]
 userVariables [
  User.RibbonIDOffset {type=NiagaraInt32}
  User.SkeletalMesh {
   type=NiagaraDataInterfaceSkeletalMesh
   defaultValue {
    struct=NiagaraExt_VariableValue_DataInterface
    value {dataInterface=&self:NiagaraDataInterfaceSkeletalMesh_0}
   }
  }
  User.VariantNum {type=NiagaraFloat, defaultValue=0, description=""}
 ]
}
