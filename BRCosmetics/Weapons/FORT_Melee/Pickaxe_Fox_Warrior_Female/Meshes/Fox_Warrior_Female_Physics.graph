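# Physics asset for the Fox Warrior Female pickaxe, defining bodies and joints for its dangling charm, straps and fur tufts.
# Capsule bodies cover the two-handed handle and each dyn_ bone, with the handle as the common anchor.
# The three-link charm chain and two-link large strap are jointed in series from the handle, while the small strap is a single link.
# Three fur tufts each attach directly to the handle with locked twist and asymmetric swing limits.
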PhysicsAsset /BRCosmetics/Weapons/FORT_Melee/Pickaxe_Fox_Warrior_Female/Meshes/Fox_Warrior_Female_Physics
bodies {
 Melee_TwoHanded_Handle [{shapeType=Capsule, length=30, radius=2.5}]
 dyn_charm_chain_1_jnt [{shapeType=Capsule, length=1, radius=1.3, center=<1.8046 0 0>}]
 dyn_charm_chain_2_jnt [{shapeType=Capsule, length=1, radius=1.3, center=<1.8046 0 0>}]
 dyn_charm_chain_3_jnt [{shapeType=Capsule, length=1, radius=2.55, center=<2.112 0.0001 -0.7866>}]
 dyn_large_fur_tuft_jnt_1 [{shapeType=Capsule, length=4, radius=2, shapeName="None, None, None", center=<1.3927 0 -0.7088>, rotation=<rot 76.2552 0.0001 0.0001>}]
 dyn_large_fur_tuft_jnt_2 [{shapeType=Capsule, length=4, radius=2, shapeName="None, None, None", center=<1.2823 -0.0515 0.2539>, rotation=<rot 31.3443 0 0>}]
 dyn_large_strap_jnt_1 [{shapeType=Capsule, length=1, radius=2, center=<2.649 0 0>}]
 dyn_large_strap_jnt_2 [{shapeType=Capsule, length=1, radius=2, center=<2.649 0 0>}]
 dyn_small_fur_tuft_jnt [{shapeType=Capsule, length=2, radius=2, shapeName="None, None, None", center=<1.363 0 -0.362>, rotation=<rot 53.958 0 0>}]
 dyn_small_strap_jnt_1 [{shapeType=Capsule, length=0.7, radius=1.2, center=<1.1311 -0.0001 0>}]
}
constraints [
 {bone1Name=dyn_large_strap_jnt_1, bone2Name=Melee_TwoHanded_Handle, swing1LimitDegrees=25, swing2LimitDegrees=30, twistLimitDegrees=5}
 {bone1Name=dyn_large_strap_jnt_2, bone2Name=dyn_large_strap_jnt_1, swing1LimitDegrees=15, swing2LimitDegrees=25, twistLimitDegrees=10}
 {bone1Name=dyn_small_strap_jnt_1, bone2Name=Melee_TwoHanded_Handle, swing1LimitDegrees=10, swing2LimitDegrees=20, twistLimitDegrees=5}
 {bone1Name=dyn_charm_chain_1_jnt, bone2Name=Melee_TwoHanded_Handle, swing1LimitDegrees=30, swing2LimitDegrees=30, twistLimitDegrees=30}
 {bone1Name=dyn_charm_chain_2_jnt, bone2Name=dyn_charm_chain_1_jnt, swing1LimitDegrees=30, swing2LimitDegrees=30, twistLimitDegrees=30}
 {bone1Name=dyn_charm_chain_3_jnt, bone2Name=dyn_charm_chain_2_jnt, swing1LimitDegrees=30, swing2LimitDegrees=30, twistLimitDegrees=30}
 {
  bone1Name=dyn_large_fur_tuft_jnt_1
  bone2Name=Melee_TwoHanded_Handle
  swing1LimitDegrees=15
  swing2LimitDegrees=40
  twistLimitDegrees=5
  twistMotion=Locked
 }
 {
  bone1Name=dyn_large_fur_tuft_jnt_2
  bone2Name=Melee_TwoHanded_Handle
  swing1LimitDegrees=10
  swing2LimitDegrees=40
  twistLimitDegrees=5
  twistMotion=Locked
 }
 {
  bone1Name=dyn_small_fur_tuft_jnt
  bone2Name=Melee_TwoHanded_Handle
  swing1LimitDegrees=10
  swing2LimitDegrees=40
  twistLimitDegrees=5
  twistMotion=Locked
 }
]
